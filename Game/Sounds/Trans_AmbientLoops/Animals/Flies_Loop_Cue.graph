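# Sound cue for a quiet, looping ambient buzz of flies, classed under the ambient tile mix.
# A looping wave player feeds a modulator that slightly randomizes pitch, on top of a raised cue-wide pitch multiplier.
# A bug VFX attenuation asset handles spatial falloff, and a concurrency override prevents new instances once the limit is reached.
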SoundCue /Game/Sounds/Trans_AmbientLoops/Animals/Flies_Loop_Cue
root=Modulator_0
nodes [
 {id=Modulator_0, children [WavePlayer_2], properties {pitchMax=1.1, pitchMin=0.9}}
 {id=WavePlayer_2, properties {bLooping=true, soundWaveAssetPtr=&/Flies002_Loop}}
]
properties {
 attenuationSettings=/Game/Sounds/Attenuation/Trans_VFX_Bugs
 bOverrideConcurrency=true
 maxDistance=700
 pitchMultiplier=1.2
 soundClassObject=/Game/Sounds/FortSoundClassesAndModes/Fort_Ambient_TileMix
 volumeMultiplier=0.15
 concurrencyOverrides {resolutionRule=PreventNew}
}
